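# Master material for the DragonPry pickaxe, adding toggleable color variants between the 2023 character parent PreFX and PostFX functions.
# Variant, Accent, Sheath, Logo and Metal layers regrade the parent's output with parameterized tint, power and gain adjustments, then blend through masks.
# Masks come from customization and tech-art textures plus a hard-coded Rash backpack FX texture for the sheath.
# An Is Backpack scalar suggests sharing with the backpack.
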Material /BRCosmetics/Weapons/FORT_Melee/Pickaxe_DragonPry/Materials/M_Pickaxe_DragonPry
graph {
 properties {
  bAutomaticallySetUsageInEditor=false
  bUseMaterialAttributes=true
  bUsedWithClothing=true
  bUsedWithMeshDeformer=true
  bUsedWithMorphTargets=true
  bUsedWithNanite=true
  bUsedWithNiagaraMeshParticles=true
  bUsedWithSkeletalMesh=true
  pixelDepthOffsetMode=PDOM_Legacy
  shadingModel=MSM_FromMaterialExpression
  usedShadingModels=Default Lit | Subsurface | Clear Coat | Subsurface Profile | Cloth
 }
 nodes [
  n1 = NamedRerouteUsage()
  n2 = BlendMaterialAttributes(A=@n1, B=NamedRerouteUsage(), Alpha=NamedRerouteUsage())
  n3 = NamedRerouteUsage()
  n4 = NamedRerouteUsage()
  n5 = BlendMaterialAttributes(
   A=BlendMaterialAttributes(A=BlendMaterialAttributes(A=@n2, B=@n3, Alpha=@n4), B=NamedRerouteUsage(), Alpha=NamedRerouteUsage())
   B=NamedRerouteUsage()
   Alpha=NamedRerouteUsage()
  )
  n6 = NamedRerouteUsage()
  n7 = NamedRerouteUsage()
  n8 = ScalarParameter() {
   controlType=Enumeration
   enumeration=/Game/Global/Materials/Enums/E_Mat_OnOff
   group=Color Variants
   parameterName=Enable Color Variants
   sliderMax=1
   sortPriority=20
  }
  n9 = MaterialFunctionCall() {materialFunction=&MF_CharacterParent_2023_PreFX}
  MaterialFunctionCall(
   Material=BlendMaterialAttributes(A=@n1, B=BlendMaterialAttributes(A=@n5, B=@n6, Alpha=@n7), Alpha=@n8)
   M=@n9.M
   Pre Skinned Local Position=@n9.M
   Distance Blend=@n9.M
   Fresnel=@n9.M
  ) {materialFunction=&MF_CharacterParent_2023_PostFX}
  NamedRerouteDeclaration(None=Multiply(
    A=TextureSample() {texture=/BRCosmetics/Accessories/FORT_Backpacks/Backpack_M_MED_DragonPry/Skins/Rash/Textures/T_DragonPry_Rash_Backpack_FX}.B
    B=NamedRerouteUsage()
   )) {name="Sheath Mask (S)", nodeColor=<rgba 1 0 0.5859 1>}
  n10 = VectorParameter() {
   group=Color Variants
   parameterName=Logo Color
   sortPriority=430
   channelNames {a=Exp, b=Tint B, g=Tint G, r=Tint R}
   defaultValue=<rgba 1 1 1 1>
  }
  n11 = NamedRerouteUsage()
  n12 = VectorParameter() {
   group=Color Variants
   parameterName=Logo Metallic
   sortPriority=450
   channelNames {a=Albedo Gain, b=Metallic Exp, g=Metallic Gain, r=Metallic Offset}
   defaultValue=<rgba 0 1 1 1>
  }
  n13 = Multiply(A=@n12.Albedo Gain, B=Multiply(A=@n10.RGB, B=Power(Base=@n11, Exp=@n10.Exp)))
  n14 = NamedRerouteUsage()
  n15 = Saturate(None=Multiply(A=Add(A=@n14, B=@n12.Metallic Offset), B=@n12.Metallic Gain))
  n16 = SetMaterialAttributes!(
   MaterialAttributes=NamedRerouteUsage()
   Base Color=NamedRerouteDeclaration(None=Saturate(None=@n13)) {name="Logo BaseColor (V3)", nodeColor=<rgba 0.3359 0 1 1>}
   Metallic=NamedRerouteDeclaration(None=Power(Base=@n15, Exp=@n12.Metallic Exp)) {name="Logo Metallic (S)", nodeColor=<rgba 0.5 1 0 1>}
   Specular=NamedRerouteDeclaration(None=NamedRerouteUsage()) {name="Logo Specular (S)", nodeColor=<rgba 1 0.1406 0 1>}
   Roughness=NamedRerouteDeclaration(None=NamedRerouteUsage()) {name="Logo Roughness (S)", nodeColor=<rgba 0 0.1797 1 1>}
  ) {attributeSetTypes [
    #1
    #2
    #3
    #4
   ]}
  NamedRerouteDeclaration(None=@n16) {name="Logo (MA)", nodeColor=<rgba 1 0.5391 0 1>}
  n17 = TextureSampleParameter2D() {
   group=05_Tech Art Masks
   parameterName=TechArtMask
   samplerSource=SSM_Wrap_WorldGroupSettings
   samplerType=SAMPLERTYPE_Masks
   sortPriority=0
   texture=&T_BlackMask
  }
  NamedRerouteDeclaration(None=@n17.G) {name="Metal Mask (S)", nodeColor=<rgba 1 0 0.5859 1>}
  n18 = GetMaterialAttributes(None=NamedRerouteUsage()) {attributeGetTypes [
    #1
    #2
    #3
    #4
    #5
    #6
   ]}
  NamedRerouteDeclaration(None=@n18.Metallic) {name="Metallic In (S)", nodeColor=<rgba 0.3826 1 0 1>}
  NamedRerouteDeclaration(None=@n18.Specular) {name="Specular In (S)", nodeColor=<rgba 0.3826 1 0 1>}
  NamedRerouteDeclaration(None=@n18.Roughness) {name="Roughness In (S)", nodeColor=<rgba 0.3826 1 0 1>}
  NamedRerouteDeclaration(None=Saturate(None=ScalarParameter() {group=Color Variants, parameterName=Is Backpack, sliderMax=1, sortPriority=1000})) {name="Is Backpack (S)", nodeColor=<rgba 1 0 0 1>}
  NamedRerouteDeclaration(None=@n9.Result) {name="Material In (MA)", nodeColor=<rgba 0.1719 1 0 1>}
  NamedRerouteDeclaration(None=@n18.Normal) {name="Normal In (S)", nodeColor=<rgba 0.3826 1 0 1>}
  n19 = VectorParameter() {
   group=Color Variants
   parameterName=Accent Color
   sortPriority=130
   channelNames {a=Exp, b=Tint B, g=Tint G, r=Tint R}
   defaultValue=<rgba 1 1 1 1>
  }
  n20 = NamedRerouteUsage()
  n21 = VectorParameter() {
   group=Color Variants
   parameterName=Accent Metallic
   sortPriority=150
   channelNames {a=Albedo Gain, b=Metallic Exp, g=Metallic Gain, r=Metallic Offset}
   defaultValue=<rgba 0 1 1 1>
  }
  n22 = Multiply(A=@n21.Albedo Gain, B=Multiply(A=@n19.RGB, B=Power(Base=@n20, Exp=@n19.Exp)))
  n23 = NamedRerouteUsage()
  n24 = Saturate(None=Multiply(A=Add(A=@n23, B=@n21.Metallic Offset), B=@n21.Metallic Gain))
  n25 = VectorParameter() {
   group=Color Variants
   parameterName=Accent Specular
   sortPriority=160
   channelNames {a=Roughness Gain, b=Roughness Exp, g=Specular Gain, r=Specular Exp}
   defaultValue=<rgba 1 1 1 1>
  }
  n26 = Multiply(A=Power(Base=NamedRerouteUsage(), Exp=@n25.Specular Exp), B=@n25.Specular Gain)
  n27 = Multiply(A=Power(Base=NamedRerouteUsage(), Exp=@n25.Roughness Exp), B=@n25.Roughness Gain)
  n28 = SetMaterialAttributes(
   MaterialAttributes=NamedRerouteUsage()
   Base Color=NamedRerouteDeclaration(None=Saturate(None=@n22)) {name="Accent BaseColor (V3)", nodeColor=<rgba 0.3359 0 1 1>}
   Metallic=NamedRerouteDeclaration(None=Power(Base=@n24, Exp=@n21.Metallic Exp)) {name="Accent Metallic (S)", nodeColor=<rgba 0.5 1 0 1>}
   Specular=NamedRerouteDeclaration(None=Saturate(None=@n26)) {name="Accent Specular (S)", nodeColor=<rgba 1 0.1406 0 1>}
   Roughness=NamedRerouteDeclaration(None=Saturate(None=@n27)) {name="Accent Roughness (S)", nodeColor=<rgba 0 0.1797 1 1>}
  ) {attributeSetTypes [
    #1
    #2
    #3
    #4
   ]}
  NamedRerouteDeclaration(None=@n28) {name="Accent (MA)", nodeColor=<rgba 1 0.5391 0 1>}
  n29 = VectorParameter() {
   group=Color Variants
   parameterName=Variant Color
   sortPriority=30
   channelNames {a=Exp, b=Tint B, g=Tint G, r=Tint R}
   defaultValue=<rgba 1 1 1 1>
  }
  n30 = NamedRerouteUsage()
  n31 = VectorParameter() {
   group=Color Variants
   parameterName=Variant Metallic
   sortPriority=50
   channelNames {a=Albedo Gain, b=Metallic Exp, g=Metallic Gain, r=Metallic Offset}
   defaultValue=<rgba 0 1 1 1>
  }
  n32 = Multiply(A=@n31.Albedo Gain, B=Multiply(A=@n29.RGB, B=Power(Base=@n30, Exp=@n29.Exp)))
  n33 = NamedRerouteUsage()
  n34 = Saturate(None=Multiply(A=Add(A=@n33, B=@n31.Metallic Offset), B=@n31.Metallic Gain))
  n35 = VectorParameter() {
   group=Color Variants
   parameterName=Variant Specular
   sortPriority=60
   channelNames {a=Roughness Gain, b=Roughness Exp, g=Specular Gain, r=Specular Exp}
   defaultValue=<rgba 1 1 1 1>
  }
  n36 = Multiply(A=Power(Base=NamedRerouteUsage(), Exp=@n35.Specular Exp), B=@n35.Specular Gain)
  n37 = Multiply(A=Power(Base=NamedRerouteUsage(), Exp=@n35.Roughness Exp), B=@n35.Roughness Gain)
  n38 = SetMaterialAttributes(
   MaterialAttributes=NamedRerouteUsage()
   Base Color=NamedRerouteDeclaration(None=Saturate(None=@n32)) {name="Variant BaseColor (V3)", nodeColor=<rgba 0.3359 0 1 1>}
   Metallic=NamedRerouteDeclaration(None=Power(Base=@n34, Exp=@n31.Metallic Exp)) {name="Variant Metallic (S)", nodeColor=<rgba 0.5 1 0 1>}
   Specular=NamedRerouteDeclaration(None=Saturate(None=@n36)) {name="Variant Specular (S)", nodeColor=<rgba 1 0.1406 0 1>}
   Roughness=NamedRerouteDeclaration(None=Saturate(None=@n37)) {name="Variant Roughness (S)", nodeColor=<rgba 0 0.1797 1 1>}
  ) {attributeSetTypes [
    #1
    #2
    #3
    #4
   ]}
  NamedRerouteDeclaration(None=@n38) {name="Variant (MA)", nodeColor=<rgba 1 0.5391 0 1>}
  n39 = NamedRerouteUsage()
  n40 = VectorParameter() {
   group=Color Variants
   parameterName=Metal Controls
   sortPriority=310
   channelNames {a=Roughness Gain, b=Roughness Exp, g=Metallic Gain, r=Albedo Exp}
   defaultValue=<rgba 1 1 1 1>
  }
  n41 = Multiply(A=NamedRerouteUsage(), B=@n40.Metallic Gain)
  n42 = Multiply(A=Power(Base=NamedRerouteUsage(), Exp=@n40.Roughness Exp), B=@n40.Roughness Gain)
  n43 = SetMaterialAttributes(
   MaterialAttributes=NamedRerouteUsage()
   Base Color=NamedRerouteDeclaration(None=Power(Base=@n39, Exp=@n40.Albedo Exp)) {name="Metal Bits BaseColor (V3)", nodeColor=<rgba 0.3359 0 1 1>}
   Metallic=NamedRerouteDeclaration(None=Saturate(None=@n41)) {name="Metal Bits Metallic (S)", nodeColor=<rgba 0.5 1 0 1>}
   Specular=NamedRerouteDeclaration(None=NamedRerouteUsage()) {name="Metal Bits Specular (S)", nodeColor=<rgba 1 0.1406 0 1>}
   Roughness=NamedRerouteDeclaration(None=Saturate(None=@n42)) {name="Metal Bits Roughness (S)", nodeColor=<rgba 0 0.1797 1 1>}
  ) {attributeSetTypes [
    #1
    #2
    #3
    #4
   ]}
  NamedRerouteDeclaration(None=@n43) {name="Metal (MA)", nodeColor=<rgba 1 0.5391 0 1>}
  NamedRerouteDeclaration(None=@n18.Emissive Color) {name="Emission In (S)", nodeColor=<rgba 0.3826 1 0 1>}
  n44 = TextureSampleParameter2D() {
   group=Color Variants
   parameterName=ColorCustomizationMask
   samplerSource=SSM_Wrap_WorldGroupSettings
   samplerType=SAMPLERTYPE_Masks
   sortPriority=0
   texture=&T_BlackMask
  }
  NamedRerouteDeclaration(None=@n44.R) {name="Variant Mask (S)", nodeColor=<rgba 1 0 0.5859 1>}
  n45 = NamedRerouteUsage()
  NamedRerouteDeclaration(None=Multiply(A=@n44.B, B=OneMinus(None=@n45))) {name="Accent Mask (S)", nodeColor=<rgba 1 0 0.5859 1>}
  n46 = VectorParameter() {
   group=Color Variants
   parameterName=Sheath Color
   sortPriority=230
   channelNames {a=Exp, b=Tint B, g=Tint G, r=Tint R}
   defaultValue=<rgba 1 1 1 1>
  }
  n47 = NamedRerouteUsage()
  n48 = VectorParameter() {
   group=Color Variants
   parameterName=Sheath Metallic
   sortPriority=250
   channelNames {a=Albedo Gain, b=Metallic Exp, g=Metallic Gain, r=Metallic Offset}
   defaultValue=<rgba 0 1 1 1>
  }
  n49 = Multiply(A=@n48.Albedo Gain, B=Multiply(A=@n46.RGB, B=Power(Base=@n47, Exp=@n46.Exp)))
  n50 = NamedRerouteUsage()
  n51 = Saturate(None=Multiply(A=Add(A=@n50, B=@n48.Metallic Offset), B=@n48.Metallic Gain))
  n52 = VectorParameter() {
   group=Color Variants
   parameterName=Sheath Specular
   sortPriority=260
   channelNames {a=Roughness Gain, b=Roughness Exp, g=Specular Gain, r=Specular Exp}
   defaultValue=<rgba 1 1 1 1>
  }
  n53 = Multiply(A=Power(Base=NamedRerouteUsage(), Exp=@n52.Specular Exp), B=@n52.Specular Gain)
  n54 = Multiply(A=Power(Base=NamedRerouteUsage(), Exp=@n52.Roughness Exp), B=@n52.Roughness Gain)
  n55 = SetMaterialAttributes(
   MaterialAttributes=NamedRerouteUsage()
   Base Color=NamedRerouteDeclaration(None=Saturate(None=@n49)) {name="Sheath BaseColor (V3)", nodeColor=<rgba 0.3359 0 1 1>}
   Metallic=NamedRerouteDeclaration(None=Power(Base=@n51, Exp=@n48.Metallic Exp)) {name="Sheath Metallic (S)", nodeColor=<rgba 0.5 1 0 1>}
   Specular=NamedRerouteDeclaration(None=Saturate(None=@n53)) {name="Sheath Specular (S)", nodeColor=<rgba 1 0.1406 0 1>}
   Roughness=NamedRerouteDeclaration(None=Saturate(None=@n54)) {name="Sheath Roughness (S)", nodeColor=<rgba 0 0.1797 1 1>}
  ) {attributeSetTypes [
    #1
    #2
    #3
    #4
   ]}
  NamedRerouteDeclaration(None=@n55) {name="Sheath (MA)", nodeColor=<rgba 1 0.5391 0 1>}
  NamedRerouteDeclaration(None=@n18.Base Color) {name="Base Color In (V3)", nodeColor=<rgba 0.3826 1 0 1>}
  NamedRerouteDeclaration(None=Multiply(A=@n44.G, B=NamedRerouteUsage())) {name="Logo Mask (S)", nodeColor=<rgba 1 0 0.5859 1>}
 ]
}
