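# Physics asset for the DragonPry pickaxe, named as its frontend variant, that gives its dangling tassels physical motion.
# Capsule bodies cover the two-handed handle and a four-bone tassel chain, each bone constrained to the one above it.
# The tassel root rotates freely on all axes against the handle, while the lower joints have moderate swing limits and very tight twist, so the chain bends without spinning.
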PhysicsAsset /BRCosmetics/Weapons/FORT_Melee/Pickaxe_DragonPry/Meshes/DragonPry_Axe_FrontendPhysics
bodies {
 Melee_TwoHanded_Handle [{shapeType=Capsule, length=33.9202, radius=2.2, shapeName=Melee_TwoHanded_Handle_capsule, center=<0 0 6>}]
 dyn_tassels_base [{shapeType=Capsule, length=0.505, radius=0.505, shapeName=dyn_tassels_base_capsule}]
 dyn_tassels_base_01 [{shapeType=Capsule, length=1, radius=1.5, shapeName=dyn_tassels_base_01_capsule, center=<2 0 0>, rotation=<rot 90 0 0>}]
 dyn_tassels_base_02 [{shapeType=Capsule, length=5, radius=1.5, shapeName=dyn_tassels_base_02_capsule, center=<5 0 0>, rotation=<rot 90 0 0>}]
 dyn_tassels_base_03 [{shapeType=Capsule, length=3, radius=1.5, shapeName=dyn_tassels_base_03_capsule, center=<5 0 0>, rotation=<rot 90 0 0>}]
}
constraints [
 {
  bone1Name=dyn_tassels_base
  bone2Name=Melee_TwoHanded_Handle
  swing1LimitDegrees=45
  swing1Motion=Free
  swing2LimitDegrees=45
  swing2Motion=Free
  twistLimitDegrees=45
  twistMotion=Free
 }
 {bone1Name=dyn_tassels_base_01, bone2Name=dyn_tassels_base, swing1LimitDegrees=40, swing2LimitDegrees=40, twistLimitDegrees=2}
 {bone1Name=dyn_tassels_base_02, bone2Name=dyn_tassels_base_01, swing1LimitDegrees=40, swing2LimitDegrees=40, twistLimitDegrees=2}
 {bone1Name=dyn_tassels_base_03, bone2Name=dyn_tassels_base_02, swing1LimitDegrees=40, swing2LimitDegrees=40, twistLimitDegrees=2}
]
