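# Anim Blueprint for the DragonPry pickaxe that adds physics-driven motion to its tassel bones.
# Additive bone-space rotations pre-pose the tassels around two chained RigidBody nodes: base-bone space with the main physics asset, then world space with a frontend physics asset.
# The event graph derives Idle, idleJump and FrontEndAlpha from movement and frontend flags, though the AnimGraph shows no visible use of them.
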AnimBlueprint /BRCosmetics/Weapons/FORT_Melee/Pickaxe_DragonPry/Meshes/DragonPry_Axe_AnimBP
target_skeleton=&/DragonPry_Axe_Skeleton
graphs [
 AnimGraph {nodes [
   AnimGraphNode_SubInput_0 = Input Pose()
   n1 = "Transform (Modify) Bone - Bone: dyn_tassels_base"(ComponentPose=Local To Component(LocalPose=@AnimGraphNode_SubInput_0), Alpha=1, Translation="0,0,0", Rotation="20,0,0", Scale="1,1,1") {
    rotationMode=BMM_Additive
    rotationSpace=BCS_BoneSpace
    translationMode=BMM_Additive
    translationSpace=BCS_BoneSpace
    alphaBoolBlend {blendInTime=2, blendOutTime=2}
    boneToModify {boneName=dyn_tassels_base}
    rotation {pitch=20}
   }
   n2 = "Transform (Modify) Bone - Bone: dyn_tassels_base_01"(ComponentPose=@n1, Alpha=1, Translation="0,0,0", Rotation="40,0,0", Scale="1,1,1") {
    rotationMode=BMM_Additive
    rotationSpace=BCS_BoneSpace
    translationMode=BMM_Additive
    translationSpace=BCS_BoneSpace
    alphaBoolBlend {blendInTime=2, blendOutTime=2}
    boneToModify {boneName=dyn_tassels_base_01}
    rotation {pitch=40}
   }
   n3 = RigidBody(ComponentPose=@n2) {
    bOverrideWorldGravity=true
    overridePhysicsAsset=&/DragonPry_Axe_Physics
    simulationSpace=BaseBoneSpace
    baseBoneRef {boneName=b_AR_Root}
    overrideWorldGravity {z=-980}
    simSpaceSettings {
     dampingAlpha=0
     maxAngularAcceleration=40
     maxAngularVelocity=8
     maxLinearAcceleration=4000
     maxLinearVelocity=800
     velocityScaleZ=0.6
     worldAlpha=1
     externalLinearDragV=<0.2 0.2 0.2>
    }
   }
   n4 = "Transform (Modify) Bone - Bone: dyn_tassels_master"!(ComponentPose=@n3, bAlphaBoolEnabled:bool=True, Translation="0,0,0", Rotation="-30,0,0", Scale="1,1,1") -> (Pose) {
    alphaInputType=Bool
    rotationMode=BMM_Additive
    rotationSpace=BCS_BoneSpace
    translationMode=BMM_Additive
    translationSpace=BCS_BoneSpace
    alphaBoolBlend {blendInTime=2, blendOutTime=2}
    boneToModify {boneName=dyn_tassels_master}
    rotation {pitch=-30}
   }
   n5 = "Transform (Modify) Bone - Bone: dyn_tassels_base_02"!(ComponentPose=@n4, bAlphaBoolEnabled:bool=True, Translation="0,0,0", Rotation="20,0,0", Scale="1,1,1") -> (Pose) {
    alphaInputType=Bool
    rotationMode=BMM_Additive
    rotationSpace=BCS_BoneSpace
    translationMode=BMM_Additive
    translationSpace=BCS_BoneSpace
    alphaBoolBlend {blendInTime=2, blendOutTime=2}
    boneToModify {boneName=dyn_tassels_base_02}
    rotation {pitch=20}
   }
   n6 = "Transform (Modify) Bone - Bone: dyn_tassels_base_03"!(ComponentPose=@n5, bAlphaBoolEnabled:bool=True, Translation="0,0,0", Rotation="4,0,0", Scale="1,1,1") -> (Pose) {
    alphaInputType=Bool
    rotationMode=BMM_Additive
    rotationSpace=BCS_BoneSpace
    translationMode=BMM_Additive
    translationSpace=BCS_BoneSpace
    alphaBoolBlend {blendInTime=2, blendOutTime=2}
    boneToModify {boneName=dyn_tassels_base_03}
    rotation {pitch=4}
   }
   n7 = RigidBody(ComponentPose=@n6) {
    bOverrideWorldGravity=true
    overridePhysicsAsset=&/DragonPry_Axe_FrontendPhysics
    simulationSpace=WorldSpace
    baseBoneRef {boneName=b_AR_Root}
    componentAppliedLinearAccClamp=<2000 2000 2000>
    overrideWorldGravity {z=-980}
    simSpaceSettings {dampingAlpha=0, maxAngularAcceleration=40, maxAngularVelocity=8, maxLinearAcceleration=4000, maxLinearVelocity=800, worldAlpha=0.1}
   }
   Output Pose(Result=Component To Local(ComponentPose=@n7)) {layerGroup=None}
  ]}
 EventGraph {nodes [
   Event BlueprintUpdateAnimation()
   > n1 = Sequence!(execute) -> (then_0, then_1, then_2:Exec)
   n2 = Get bIsMoving2D()
   n3 = Get bIsSprinting()
   n4 = Get bIsCrouching()
   > Set Idle(Idle:bool=NOT Boolean(A=OR Boolean!(A=@n2, B=@n4, C:bool=@n3) -> (ReturnValue))) -> (Output_Get:bool)
   n5 = Branch(execute=@n1.then_1, Condition=OR Boolean!(A=Get bIsFalling(), B=Get bIsJumping()) -> (ReturnValue))
   > Set idleJump(idleJump:double=1) -> (Output_Get:double)
   Set idleJump(execute=@n5.else, idleJump=0)
   n6 = Branch(execute=@n1.then_2, Condition=Get bIsInFrontEnd())
   > Set FrontEndAlpha(FrontEndAlpha:double=0.5) -> (Output_Get:double)
   Set FrontEndAlpha(execute=@n6.else, FrontEndAlpha=0)
   TryGetPawnOwner()
  ]}
]
variables [
 FrontEndAlpha {type=number}
 GravityOverride {type=Vector}
 Idle {type=boolean}
 idleJump {type=number}
 RBAN Settings Out {type=FortRigidBodyAnimNodeParameters}
]
